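; Deep Dish Diffusal actor for the Ashton Chicago backpack item, supplying shield block feedback when a damage gameplay cue fires.
; The cue handler executes a local block cue at the impact point, then, where cosmetics run, spawns a reflect impact emitter and activates a ripple.
; PlayHitReactMontage plays a random owner montage section matched to hit yaw and first-block state (1.5s since last hit), unless one is already playing.
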(event GameplayCue|GameplayCue.Damage
  (params (EventType :type EGameplayCueEvent%20Enum :default "OnActive") (Parameters :type Gameplay%20Cue%20Parameters%20Structure%20%28by%20ref%29))
  (bind (_normalizedmagnitude _rawmagnitude _effectcontext) (Ability|GameplayCue|BreakGameplayCueParameters Parameters))
  (bind (_returnvalue_roll _returnvalue_pitch _returnvalue_yaw) (Math|Rotator|BreakRotator (Transformation|GetWorldRotation (Variables|Default|GetBox))))
  (bind _returnvalue (Ability|EffectContext|GetHitResult _effectcontext))
  (bind (_bblockinghit _binitialoverlap _time _distance _location _impactpoint _normal _impactnormal) (Collision|BreakHitResult _returnvalue))
  ;[1 1 1] Proto Impact FX logic
  (GameplayCue|ExecuteGameplayCue%28Burst%29Local self (Variables|Default|GetBlockActiveCue) (Ability|GameplayCue|MakeGameplayCueParameters :NormalizedMagnitude 0.0 :RawMagnitude 0.0 :Location _impactpoint :Normal _impactnormal :SourceObject self :GameplayEffectLevel 1 :AbilityLevel 1 :bReplicateLocationWhenUsingMinimalRepProxy false) false)
  (Networking|CanExecuteCosmeticEvents
    (:True
      (Effects|SpawnEmitteratLocation "/Game/Athena/Items/Gameplay/BackPacks/Ashton/Chicago/FX_Assets/P_Ashton_Chicago_ShieldReflectImpact.P_Ashton_Chicago_ShieldReflectImpact" _impactpoint (Math|Rotator|Lerp(Rotator) (Math|Rotator|MakeRotator _returnvalue_roll (Math|Float|Lerp 0.0 3.6e+02 (Math|Random|RandomFloat)) _returnvalue_yaw) (Math|Conversions|RotationFromXVector _impactnormal) 0.1))
      (Components|Activation|Activate (Variables|Default|GetP_Ashton_Chicago_ShieldReflectRipple) true)
      ;[1 1 1] Play hit reaction on owner
      (CallFunction|PlayHitReactMontage _returnvalue)
      (Variables|Default|SetLastHitTime (Utilities|Time|GetGameTimeinSeconds)))))

(fn PlayHitReactMontage
  (params (HitResult :type Hit%20Result%20Structure))
  (bind _currentselection (LocalVariables|GetCurrentSelection))
  (bind _returnvalue (Utilities|Array|Length _currentselection))
  (bind (_bblockinghit _binitialoverlap _time _distance _location _impactpoint _normal _impactnormal _physmat _hitactor _hitcomponent _hitbonename _bonename _hititem _elementindex _faceindex _tracestart _traceend) (Collision|BreakHitResult HitResult))
  ;[1 1 1] Get Owner info
  (bind _asfort_player_pawn (Utilities|Casting|CastToFortPlayerPawn (Actor|GetOwner)))
  ;[1 1 1] Don't play if a reaction is already playing
  (if (not (Animation|Montage|MontageIsPlaying (Components|SkeletalMesh|GetAnimInstance (Variables|/Script/Engine.Character|GetMesh _asfort_player_pawn)) (Variables|Default|GetHitReactionMontage)))
    ;[1 1 1] Build list of valid section names
    (for _array_element (Variables|Default|GetHitReactSelectionInfos)
      (if (> (- (Utilities|Time|GetGameTimeinSeconds) (Variables|Default|GetLastHitTime)) 1.5)
        (if (.IsFirstBlock_25_919B150F474766AEAD087EA3A84DC138 (Utilities|Struct|BreakDishHitReactSelectionInfo _array_element))
          _)
        (elif (.IsFirstBlock_25_919B150F474766AEAD087EA3A84DC138 (Utilities|Struct|BreakDishHitReactSelectionInfo _array_element))
          _))
      (if (Math|Float|InRange(Float) (Math|Trig|RadiansToDegrees (Math|Vector|VectorHeadingAngle (Math|Vector|UnrotateVector (- _tracestart _traceend) (Transformation|GetActorRotation _asfort_player_pawn)))) (Utilities|Struct|BreakDishHitReactSelectionInfo _array_element) (.MaxYawToInstigator_9_62A6FC5442FA6451356E01968FB0A85F (Utilities|Struct|BreakDishHitReactSelectionInfo _array_element)))
        (Utilities|Array|Add (LocalVariables|GetCurrentSelection) (.MontageSectionName_17_C1FAF635407E1271A9D38598E5247255 (Utilities|Struct|BreakDishHitReactSelectionInfo _array_element)))))
    ;[1 1 1] Pick a random section from the valid list
    (if (> _returnvalue 0)
      ;[1 1 1] Play the animation
      (Animation|PlayAnimMontage _asfort_player_pawn (Variables|Default|GetHitReactionMontage) 1.0 (Utilities|Array|Get(aref) (LocalVariables|GetCurrentSelection) (Math|Random|RandomIntegerinRange 0 (- _returnvalue 1)))))))

(fn ConstructionScript)
