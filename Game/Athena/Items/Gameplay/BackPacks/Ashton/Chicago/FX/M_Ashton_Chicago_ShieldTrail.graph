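# Additive, unlit, two-sided trail material for the Ashton Chicago backpack's shield effect, usable on Niagara ribbons, sprites and beam trails.
# Color blends level sunlight with a parameter, optionally scaled by particle color, then passes through an emissive adjustment function.
# Opacity combines panning UV1 noise, particle alpha, U-based trail masks, optional V edge softening, angle fade and switchable depth fade.
# A world position offset tilts ribbon edges separately by V, scaled by particle green.
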Material /Game/Athena/Items/Gameplay/BackPacks/Ashton/Chicago/FX/M_Ashton_Chicago_ShieldTrail
graph {
 properties {
  bUsedWithBeamTrails=true
  bUsedWithNiagaraRibbons=true
  bUsedWithNiagaraSprites=true
  blendMode=BLEND_Additive
  pixelDepthOffsetMode=PDOM_Legacy
  shadingModel=MSM_Unlit
  twoSided=true
  usedShadingModels=Unlit
 }
 nodes [
  n1 = ParticleColor()
  n2 = Multiply(A=VectorParameter() {parameterName=RibbonTiltTop, defaultValue=<rgba -50 0 -50 0>}.RGB, B=@n1.G)
  n3 = Multiply(A=@n1.G, B=VectorParameter() {parameterName=RibbonTiltBottom, defaultValue=<rgba 50 0 50 0>}.RGB)
  n4 = Round(None=ComponentMask(None=TextureCoordinate()) {g=true})
  LinearInterpolate(
   B=LinearInterpolate(A=Transform(None=@n2), B=Transform(None=@n3), Alpha=Saturate(None=@n4))
   Alpha=ScalarParameter() {defaultValue=1, parameterName=Use WPO}
  )
  n5 = LinearInterpolate(
   A=ComponentMask(None=CollectionParameter() {collection=&FortniteMaterialParameters, parameterName=SunLightColor}) {b=true, g=true, r=true}
   B=VectorParameter() {group=Color, parameterName=Color, sortPriority=2, defaultValue=<rgba 1 1 1 1>}.RGB
   Alpha=ScalarParameter() {defaultValue=1, group=Color, parameterName=Use Level Sunlight Color or Color, sortPriority=1}
  )
  n6 = ParticleColor()
  MaterialFunctionCall(Emissive=LinearInterpolate(A=@n5, B=Multiply(A=@n5, B=@n6.RGB), Alpha=ScalarParameter() {group=Color, parameterName=Use Particle Color Scale})) {materialFunction=&MF_FX_EmissiveAdjustments}
  n7 = VertexNormalWS()
  n8 = CameraVectorWS()
  n9 = Saturate(None=Power(Base=DotProduct(A=@n7, B=@n8), Exp=ScalarParameter() {defaultValue=1, group=Opacity, parameterName=Angle Fade Power}) {constExponent=6})
  NamedRerouteDeclaration(None=If(A=ScalarParameter() {group=Opacity, parameterName=Use Angle Fade}, "A > B"=@n9, "A == B"=@n9, "A < B"=Constant() {r=1}) {constB=0.99}) {name=AngleFade, nodeColor=<rgba 1 0 0.4922 1>}
  n10 = TextureCoordinate() {coordinateIndex=1}
  n11 = AppendVector(
   A=ScalarParameter() {defaultValue=1, group=Noise, parameterName=NoiseUtliing, sortPriority=3}
   B=ScalarParameter() {defaultValue=1, group=Noise, parameterName=NoiseVtliing, sortPriority=4}
  )
  n12 = ScalarParameter() {defaultValue=0.5, group=Noise, parameterName=PanSpeedU, sortPriority=5}
  n13 = ScalarParameter() {group=Noise, parameterName=PanSpeedV, sortPriority=6}
  n14 = TextureSampleParameter2D(UVs=Panner(Coordinate=Multiply(A=@n10, B=@n11), Speed=AppendVector(A=@n12, B=@n13)) {speedX=0.5}) {group=Noise, parameterName=Noise, samplerType=SAMPLERTYPE_Grayscale, texture=&T_Smoke_Trail_Soft}
  n15 = ChannelMaskParameter(None=@n14.RGBA) {group=Noise, maskChannel=Green, parameterName=Noise Mask, sortPriority=1, defaultValue=<rgba 0 1 0 0>}
  n16 = ScalarParameter() {defaultValue=1, desc=Add Mask texture intensity, group=Noise, parameterName=Noise_Intensity, sortPriority=2}
  n17 = Multiply(A=@n6.A, B=Clamp(None=Multiply(A=@n15.R, B=@n16)))
  n18 = LinearInterpolate(B=ComponentMask(None=TextureCoordinate()) {g=true}, Alpha=ScalarParameter() {group=Hard Edge, parameterName=Remove HardEdge}) {constA=1}
  n19 = MaterialFunctionCall(Float2=TextureCoordinate()) {materialFunction=&BreakOutFloat2Components}
  n20 = Power(Base=OneMinus(None=@n19.R), Exp=ScalarParameter() {defaultValue=2, group=Opacity Mask, parameterName=Trailing Mask Contrast, sortPriority=2})
  n21 = Power(Base=@n19.R, Exp=ScalarParameter() {defaultValue=2, group=Opacity Mask, parameterName=Leading Edge Mask Contrast, sortPriority=3})
  n22 = Clamp(None=Multiply(A=Multiply(A=@n20, B=@n21), B=ScalarParameter() {defaultValue=1, group=Opacity Mask, parameterName=Equalize Powers, sortPriority=1}))
  n23 = ScalarParameter() {group=Opacity Mask, parameterName=Mask Leading Edge}
  n24 = Multiply(A=Multiply(A=Multiply(A=@n17, B=@n18), B=LinearInterpolate(A=@n20, B=@n22, Alpha=@n23)), B=NamedRerouteUsage())
  StaticSwitchParameter(True=DepthFade(Opacity=@n24, FadeDistance=ScalarParameter() {defaultValue=100, parameterName=DepthFadeDist}), False=@n24) {parameterName=UseDepthFade}
 ]
}
