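; Gameplay ability for the Ashton Chicago backpack item that picks the character's catch animation when activated.
; It chooses the in-motion catch montage when the pawn is crouched, on the ground, moving above a small speed threshold and not sprinting, otherwise the normal catch montage.
; The choice is stored in the montage-to-play variable before the parent ability's activation runs, leaving the rest of the catch to the parent.
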(event Ability|EventActivateAbility
  (Variables|CatchVars|SetMontageToPlay (select (Math|Boolean|ANDBoolean (Variables|/Script/Engine.Character|GetbIsCrouched (Utilities|Casting|CastToFortPlayerPawn (Ability|GetAvatarActorfromActorInfo))) (AI|Components|NavMovement|IsMovingonGround (Variables|/Script/Engine.Character|GetCharacterMovement (Utilities|Casting|CastToFortPlayerPawn (Ability|GetAvatarActorfromActorInfo)))) (>= (Math|Vector|VectorLength (Transformation|GetVelocity (Utilities|Casting|CastToFortPlayerPawn (Ability|GetAvatarActorfromActorInfo)))) 1e+01) (not (PawnStats|IsSprinting (Utilities|Casting|CastToFortPlayerPawn (Ability|GetAvatarActorfromActorInfo))))) (Variables|Default|GetCatchInMotionCMM) (Variables|Default|GetNormalCatchCMM)))
  (|CallParentFunction))
