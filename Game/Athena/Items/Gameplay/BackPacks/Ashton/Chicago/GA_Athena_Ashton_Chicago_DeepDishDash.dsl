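; Gameplay ability for the Ashton Chicago backpack item that, despite its Dash name, runs a shield defense on the player.
; Activation applies a shield-tag effect on authority, forcing net updates for cue-driven animation, and attaches a spawned DeepDishDiffusal actor in front of the pawn.
; A defense-end gameplay event or being downed ends it, removing actor and tag.
; Leftover dash logic, mostly marked unused, capsule-traces to launch pawns and destroy buildings.
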(event Custom|TravelingComplete
  ;[1 0.2957 0.1153] UNUSED - Dash Ability Logic
  (Utilities|Time|ClearandInvalidateTimerbyHandle (Variables|BuildingTrace|GetTraceForBuildingsHandle))
  (Ability|Animation|MontageStop -1.0)
  (Input|EnableInput (Variables|Default|GetPlayerPawn) (Fortnite|GetFortPlayerControllerfromActor (Variables|Default|GetPlayerPawn)))
  (Ability|Tasks|PlayMontageandWaitforTrigger "None" "/Game/Animation/Game/MainPlayer/Combat/Gadgets/Medium/Male/Ashton/Chicago/AshtonChicago_OnCoolDown_M.AshtonChicago_OnCoolDown_M"
    (:Cancelled
      _)
    (:Completed
      _))
  (Ability|EndAbility))

(event Custom|TraceForBuildings
  ;[1 0.2957 0.1153] UNUSED - Dash Ability Logic
  ;;[1 1 1] Dash
  (bind _hitsarray (CallFunction|TargetTrace))
  (CallFunction|HandleHits _hitsarray (Variables|Default|GetPlayerPawn)))

(event Custom|HandleHits
  (params (HitsArray :type Array%20of%20Hit%20Result%20Structures) (PlayerPawn :type Fort%20Player%20Pawn%20Object%20Reference))
  ;[1 0.2957 0.1153] UNUSED - Dash Ability Logic
  ;;[1 1 1] Dash
  (CallFunction|DestroyBuilding HitsArray PlayerPawn))

(event Ability|EventOnEndAbility
  (params (bWasCancelled :type Boolean))
  ;[1 1 1] Stop montage, remove tag and failsafe shield destroy on end
  (Actor|DestroyActor (Variables|Default|GetActorRef))
  (Ability|Animation|MontageStop -1.0)
  (if (Networking|HasAuthority (Variables|Default|GetPlayerPawn))
    (Ability|RemoveGameplayEffectFromOwnerWithGrantedTags (Variables|Default|GetShieldActiveTag)) ; has to go here because only the client gets the event
    (Networking|ForceNetUpdate (Variables|Default|GetPlayerPawn)) ; Doing this because we're using a GC tag to run the anim.\nThis forces the GC through quickly for the other clients to see.
  ))

(event Ability|EventActivateAbility
  ;[1 1 1] Begin - Spawn and attach Shield
  (|CallParentFunction)
  (bind _asfort_player_pawn (Utilities|Casting|CastToFortPlayerPawn (Ability|GetAvatarActorfromActorInfo)))
  (bind _output_get (Variables|Default|SetPlayerPawn _asfort_player_pawn))
  (EventDispatchers|AssignOnPlayerStartDBNO _output_get (delegate Custom|DownedWhileDefending))
  (if (Networking|HasAuthority (Variables|Default|GetPlayerPawn))
    (Ability|ApplyGameplayEffectToOwner (Variables|Default|GetShieldActiveTagGE))
    (Networking|ForceNetUpdate (Variables|Default|GetPlayerPawn)) ; Doing this because we're using a GC tag to run the anim.\nThis forces the GC through quickly for the other clients to see.
    (else
      _))
  (Ability|Tasks|WaitGameplayEvent "(TagName=\"Event.Athena.Item.Ashton.Chicago.Defense.End\")"
    (:then
      (bind _returnvalue (Game|SpawnActorfromClass "/Game/Athena/Items/Gameplay/BackPacks/Ashton/Chicago/B_Ashton_Chicago_DeepDishDiffusal.B_Ashton_Chicago_DeepDishDiffusal_C" (Math|Transform|MakeTransform (+ (+ (Transformation|GetActorLocation (Variables|Default|GetPlayerPawn)) (Variables|Default|GetBlockingVolumeZAxisOffset)) (* (Transformation|GetActorForwardVector (Variables|Default|GetPlayerPawn)) 1e+02)) (Transformation|GetActorRotation (Variables|Default|GetPlayerPawn))) "AlwaysSpawn" "OverrideRootScale" (Variables|Default|GetPlayerPawn) 0 (Variables|Default|GetPlayerPawn)))
      (Variables|Default|SetActorRef _returnvalue)
      (Transformation|AttachActorToActor (Variables|Default|GetActorRef) (Variables|Default|GetPlayerPawn) "" "KeepWorld" "KeepWorld" "KeepRelative" true))
    (:EventReceived
      (Actor|DestroyActor (Variables|Default|GetActorRef))
      (Ability|EndAbility))))

(event Custom|DownedWhileDefending
  ;[1 1 1] Begin - Spawn and attach Shield
  (Ability|EndAbility))

(fn DestroyBuilding
  (params (BuildingActors :type Array%20of%20Hit%20Result%20Structures) (PlayerPawn :type Fort%20Player%20Pawn%20Object%20Reference))
  (bind _hitactor (Variables|Default|GetHitActor))
  (bind _returnvalue (Transformation|GetActorLocation _hitactor))
  (bind _hitactor_2 (Variables|Default|GetHitActor))
  (Variables|Default|SetPlayerPawn PlayerPawn) ; Have to re-set this or it returns none on other clients
  (for _array_element BuildingActors
    (if (Variables|/Script/Engine.Actor|GetbCanBeDamaged (.HitActor (Collision|BreakHitResult _array_element)))
      (Variables|Default|SetHitActor (.HitActor (Collision|BreakHitResult _array_element)))
      (bind _asfort_player_pawn (Utilities|Casting|CastToFortPlayerPawn (.HitActor (Collision|BreakHitResult _array_element)))
        (:then
          (Utilities|Array|AddUnique (Variables|Default|GetActorsToIgnore) _asfort_player_pawn)
          (Character|LaunchCharacter _asfort_player_pawn (* (+ (Math|Vector|Normalize (- _returnvalue (Transformation|GetActorLocation (Variables|Default|GetPlayerPawn)))) (Variables|Default|GetLaunchZ)) (Variables|Default|GetLaunchMultiplier)) true true)
          ;[1 1 1] FX
          (bind _asfort_player_pawn_2 (Utilities|Casting|CastToFortPlayerPawn _hitactor))
          (Utilities|FlowControl|Sequence
            (:then_0
              (Effects|SpawnEmitterAttached (Variables|Default|GetImpactFX_OnTarget) (Variables|/Script/Engine.Character|GetMesh _asfort_player_pawn_2) "FX_Chest")
              (Camera|PlayWorldCameraShake "/Game/Effects/Fort_Effects/Camera_Effects/BP_CameraShake_Ashton_Chicago_Shield.BP_CameraShake_Ashton_Chicago_Shield_C" (Transformation|GetWorldLocation (Variables|/Script/Engine.Character|GetMesh _asfort_player_pawn_2)) 0.0 5e+02)
              (Development|PrintString "ImpactFX")
              (Audio|PlaySoundatLocation (Variables|Default|GetImpactSFX) _returnvalue) ; Play Distant Explosion
            )
            (:then_1
              (Effects|SpawnEmitterAttached (Variables|Default|GetImpactFX_OnShield) (Variables|/Script/Engine.Actor|GetRootComponent (Variables|/Script/FortniteGame.FortPawn|GetCurrentWeapon (Variables|Default|GetPlayerPawn)))))))
        (:CastFailed
          (bind _bdidactorhavetags (Class|AthenaQuestFunctionLibrary|AthenaCheckActorHasTags (.HitActor (Collision|BreakHitResult _array_element)) "(GameplayTags=((TagName=\"Deimos.Spawner\")))"))
          (if (not _bdidactorhavetags)
            (Utilities|Casting|CastToBuildingActor _hitactor_2
              (:then
                _)
              (:CastFailed
                (Utilities|Casting|CastToBuildingProp _hitactor_2
                  (:then
                    _)
                  (:CastFailed
                    (Utilities|Casting|CastToAthena_Prop_ParentBuildingContainerBlueprint _hitactor_2)))))
            (Utilities|Casting|CastToAthenaBarrierObjective (Variables|Default|GetHitActor)
              (:CastFailed
                (GameplayEffects|ApplyGameplayEffectToTarget :self (Ability|GetAbilitySystemComponent (Variables|Default|GetPlayerPawn)) :GameplayEffectClass (Variables|Default|GetBuildingDestroyGE) :Target (Ability|GetAbilitySystemComponent (Variables|Default|GetHitActor)) :Level 0.0)
                (Audio|PlaySoundatLocation "/Game/Sounds/Fort_Weapons/Axes/Axe_T04_HeavyAttack_Cue.Axe_T04_HeavyAttack_Cue" (Transformation|GetActorLocation (Variables|Default|GetHitActor)))
                (Game|Feedback|ClientPlayForceFeedback (Fortnite|GetFortPlayerControllerfromActor (Variables|Default|GetPlayerPawn)) "/Game/Weapons/ForceFeedback/Explosions/FF_Generic_Explosion_Far.FF_Generic_Explosion_Far" "None" false false false)))))))))

(fn HitFX
  (params (ImpactPoint :type Vector) (ImpactNormal :type Vector))
  (Effects|SpawnEmitteratLocation (Variables|Default|GetImpactFX_OnShield) ImpactNormal)
  (Audio|PlaySoundatLocation (Variables|Default|GetImpactSFX) ImpactPoint)
  (Fortnite|PlayLocalForceFeedbackatLocation "/Game/Weapons/ForceFeedback/Explosions/FF_Generic_Explosion_Near.FF_Generic_Explosion_Near" ImpactPoint 4.5e+02 "Explosion"))

(fn OnRep_HitActor)

(fn SetupAbility
  (Variables|Default|SetRunDuration (/ (Variables|Default|GetRushDistance) (Variables|Default|GetRushAnimDistancePerSecond))))

(fn TargetTrace
  (bind _outhits (Collision|MultiCapsuleTraceForObjects (+ (Transformation|GetActorLocation (Variables|Default|GetPlayerPawn)) (Variables|BuildingTrace|GetBuildingTraceZOffset)) (+ (+ (Transformation|GetActorLocation (Variables|Default|GetPlayerPawn)) (Variables|BuildingTrace|GetBuildingTraceZOffset)) (* (+ (Transformation|GetVelocity (Variables|Default|GetPlayerPawn)) (Variables|BuildingTrace|GetBuildingTraceZOffset)) (* (Utilities|Time|GetWorldDeltaSeconds) (Variables|BuildingTrace|GetFramesAheadToTrace)))) (Variables|BuildingTrace|GetCapsuleRadius) (Variables|BuildingTrace|GetCapsuleHalfHeight) (Variables|BuildingTrace|GetDestroyObjectTypes) false (Variables|Default|GetActorsToIgnore) "ForDuration"))
  (return _outhits))
